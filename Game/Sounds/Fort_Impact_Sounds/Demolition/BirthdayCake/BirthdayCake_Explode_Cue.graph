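# Sound cue for a birthday cake explosion, likely a demolition impact sound, built from a single crowd explosion wave.
# The wave player is the root, with no randomization, mixing or modulation nodes.
# It uses far explosion attenuation and the large explosives sound class, with boosted volume and a long audible range.
# A concurrency override prevents new instances from starting once the limit is reached rather than cutting off playing ones.
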SoundCue /Game/Sounds/Fort_Impact_Sounds/Demolition/BirthdayCake/BirthdayCake_Explode_Cue
root=WavePlayer_2
nodes [{id=WavePlayer_2, properties {soundWaveAssetPtr=&/BirthdayCake_Explode_Crowd_01}}]
properties {
 attenuationSettings=/Game/Sounds/Attenuation/Fort_Explosion_Far_StW_Att
 bOverrideConcurrency=true
 maxDistance=11000
 soundClassObject=/Game/Sounds/FortSoundClassesAndModes/Fort_Explosives_Large_PSM
 volumeMultiplier=1.6
 concurrencyOverrides {resolutionRule=PreventNew}
}
