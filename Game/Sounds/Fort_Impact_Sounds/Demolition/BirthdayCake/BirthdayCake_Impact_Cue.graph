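# Sound cue for impacts on the BirthdayCake object in the demolition impact sound set.
# A Random node picks one of three equally weighted impact waves, which passes through a Modulator at default settings.
# The cue plays in the Fort_Impacts sound class at reduced volume, with small-object demolition hit attenuation and the shared object-impact concurrency set.
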SoundCue /Game/Sounds/Fort_Impact_Sounds/Demolition/BirthdayCake/BirthdayCake_Impact_Cue
root=Modulator_8
nodes [
 {id=Modulator_8, children [Random_3]}
 {id=Random_3, children [WavePlayer_2,WavePlayer_5,WavePlayer_8], properties {weights [1,1,1]}}
 {id=WavePlayer_2, properties {soundWaveAssetPtr=&/BirthdayCake_Impact_01}}
 {id=WavePlayer_5, properties {soundWaveAssetPtr=&/BirthdayCake_Impact_02}}
 {id=WavePlayer_8, properties {soundWaveAssetPtr=&/BirthdayCake_Impact_03}}
]
properties {
 attenuationSettings=/Game/Sounds/Attenuation/Fort_Demolition_Hit_Small_Sized_Object_Att
 maxDistance=4000
 soundClassObject=&Fort_Impacts
 volumeMultiplier=0.64
 concurrencyOverrides {maxCount=3, resolutionRule=StopOldest}
 concurrencySet [&CONC_Object_Impacts_FourMax]
}
